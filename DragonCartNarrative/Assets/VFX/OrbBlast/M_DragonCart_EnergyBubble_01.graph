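# Translucent unlit energy bubble material for Niagara mesh particles, part of the DragonCart OrbBlast VFX.
# Two independently tiled and panned turbulence textures have their green channels summed and divided, then tinted by particle color for the glow.
# The same noise drives opacity with particle alpha, an inverted Fresnel that fades silhouette edges, and depth fade.
# A dynamic parameter sets the Fresnel base reflect fraction per particle, leaving the like-named scalar parameter unconnected.
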Material /DragonCartNarrative/Assets/VFX/OrbBlast/M_DragonCart_EnergyBubble_01
graph {
 properties {
  bAutomaticallySetUsageInEditor=false
  bUsedWithNiagaraMeshParticles=true
  blendMode=BLEND_Translucent
  shadingModel=MSM_Unlit
  usedShadingModels=Unlit
 }
 nodes [
  n1 = ParticleColor()
  n2 = ScalarParameter() {defaultValue=1, group=Emissive, parameterName=Emissive_U_Freq_01}
  n3 = ScalarParameter() {defaultValue=1, group=Emissive, parameterName=Emissive_V_Freq_01}
  n4 = Panner(
   Coordinate=Multiply(A=TextureCoordinate(), B=AppendVector(A=@n2, B=@n3))
   Speed=AppendVector(
    A=ScalarParameter() {group=Emissive, parameterName=Emissive_U_Speed_01}
    B=ScalarParameter() {group=Emissive, parameterName=Emissive_V_Speed_01}
   )
  )
  n5 = ScalarParameter() {defaultValue=1, group=Emissive, parameterName=Emissive_U_Freq_02}
  n6 = ScalarParameter() {defaultValue=1, group=Emissive, parameterName=Emissive_V_Freq_02}
  n7 = Panner(
   Coordinate=Multiply(A=TextureCoordinate(), B=AppendVector(A=@n5, B=@n6))
   Speed=AppendVector(
    A=ScalarParameter() {group=Emissive, parameterName=Emissive_U_Speed_02}
    B=ScalarParameter() {group=Emissive, parameterName=Emissive_V_Speed_02}
   )
  )
  n8 = Divide(A=Add(
    A=TextureSampleParameter2D(UVs=@n4) {group=Emissive, parameterName=Emissive_Texture_01, texture=/Game/Global/Textures/Default/Caustics/T_Turbulence_Seamless}.G
    B=TextureSampleParameter2D(UVs=@n7) {group=Emissive, parameterName=Emissive_Texture_02, texture=&T_Turbulence_Seamless}.G
   ))
  n9 = Multiply(A=@n8, B=ScalarParameter() {defaultValue=1, group=Opacity, parameterName=Opacity_Mult})
  n10 = OneMinus(None=Fresnel(
    ExponentIn=ScalarParameter() {defaultValue=5, group=Opacity, parameterName=ExponentIn}
    BaseReflectFractionIn=DynamicParameter() {paramNames [BaseReflectFractionIn, Param2, Param3, Param4]}.BaseReflectFractionIn
   ))
  DepthFade(
   Opacity=Multiply(A=@n1.A, B=Multiply(A=@n9, B=@n10))
   FadeDistance=ScalarParameter() {defaultValue=100, group=Opacity, parameterName=DepthFade}
  )
  Multiply(A=@n1.RGB, B=@n8)
  ScalarParameter() {defaultValue=0.04, group=Opacity, parameterName=BaseReflectFractionIn}
 ]
}
