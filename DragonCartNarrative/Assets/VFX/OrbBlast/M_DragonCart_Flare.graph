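# M_DragonCart_Flare is an unlit, translucent, two-sided Niagara sprite material for flare visuals in DragonCart's OrbBlast VFX.
# It multiplies particle color by a swappable flare texture, a spiky-ball flare by default, to produce the color.
# Opacity combines particle alpha with the texture's green channel raised to an adjustable power, then passes through a depth fade to soften intersections with scene geometry.
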Material /DragonCartNarrative/Assets/VFX/OrbBlast/M_DragonCart_Flare
graph {
 properties {
  bAutomaticallySetUsageInEditor=false
  bUsedWithNiagaraSprites=true
  blendMode=BLEND_Translucent
  shadingModel=MSM_Unlit
  twoSided=true
  usedShadingModels=Unlit
 }
 nodes [
  n1 = ParticleColor()
  n2 = TextureSampleParameter2D() {parameterName=Flare_Texture, texture=/Game/Effects/Fort_Effects/Textures/Flares/T_Flare_Spikeball}
  n3 = ScalarParameter() {defaultValue=1, group=Opacity, parameterName=Opacity_Power}
  DepthFade(
   Opacity=Multiply(A=@n1.A, B=Power(Base=@n2.G, Exp=@n3))
   FadeDistance=ScalarParameter() {defaultValue=100, group=Opacity, parameterName=DepthFade}
  )
  Multiply(A=@n1.RGB, B=@n2.RGB)
 ]
}
