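# Translucent, unlit, two-sided material for Niagara mesh particles, named for the DragonCart orb blast effect.
# Particle color RGB passes through the MF_FX_EmissiveAdjustments function to form the glow.
# Opacity multiplies particle alpha by an inverted Fresnel that softens silhouette edges and a powered, inverted V-coordinate gradient, then passes through DepthFade to soften intersections with scene geometry.
# It samples no textures, and scalar parameters expose the Fresnel, gradient and fade controls.
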Material /DragonCartNarrative/Assets/VFX/OrbBlast/M_DragonCart_OrbBlast_01
graph {
 properties {
  bAutomaticallySetUsageInEditor=false
  bUsedWithNiagaraMeshParticles=true
  blendMode=BLEND_Translucent
  shadingModel=MSM_Unlit
  twoSided=true
  usedShadingModels=Unlit
 }
 nodes [
  n1 = ParticleColor()
  n2 = OneMinus(None=Fresnel(
    ExponentIn=ScalarParameter() {defaultValue=5, group=Opacity, parameterName=ExponentIn}
    BaseReflectFractionIn=ScalarParameter() {defaultValue=0.04, group=Opacity, parameterName=BaseReflectFractionIn}
   ))
  n3 = TextureCoordinate()
  n4 = Power(Base=OneMinus(None=ComponentMask(None=@n3) {g=true}), Exp=ScalarParameter() {defaultValue=2, group=Opacity, parameterName=G_Power})
  DepthFade(
   Opacity=Multiply(A=@n1.A, B=Multiply(A=@n2, B=@n4))
   FadeDistance=ScalarParameter() {defaultValue=100, group=Opacity, parameterName=DepthFade}
  )
  MaterialFunctionCall(Emissive=@n1.RGB) {materialFunction=&MF_FX_EmissiveAdjustments}
 ]
}
