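# Looping Niagara effect for the DragonCart orb blast: a flickering energy beam with glow, flares, a pulsing light, downward rays and lightning.
# Persistent burst emitters form the beam, flares and light; rate emitters add rays, lightning and streaks, some culled by distance.
# User parameters drive length, intensity and offsets, while System.OrbWPOLocal shifts many elements with the orb.
# Small lightning uses five spawn groups and mesh renderers with separate position and orientation attributes around the orb.
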NiagaraSystem /DragonCartNarrative/Assets/VFX/OrbBlast/NS_DragonCart_OrbBlast_Looping_V01
dependencies {
 usedDataInterfaces [
  &NiagaraDataInterfaceCurve
  &NiagaraDataInterfaceMeshRendererInfo
  &NiagaraDataInterfaceVectorCurve
  &NiagaraDataInterfaceCamera
 ]
 usedDynamicInputs [
  &self:ScratchDynamicInput
  &TransformVector
 ]
 usedModules [
  &self:SystemSpawnScript.SystemScriptSource.SystemScriptGraph.NiagaraNodeAssignment_0.SetVariables_#1
  &self:SystemSpawnScript.SystemScriptSource.SystemScriptGraph.NiagaraNodeAssignment_1.SetVariables_#2
  &SystemState
 ]
}
emitters {
 Beam_Core {
  emitterUpdateScript [
   &EmitterState(Life Cycle Mode=::NewEnumerator1, Loop Behavior=::NewEnumerator1, Loop Duration Mode=::NewEnumerator1, UseLoopDelay=false)
   &SpawnBurst_Instantaneous(Use Spawn Probability=false, UseLoopCountLimit=false)
  ]
  particleSpawnScript [
   &InitializeParticle(
    Color=<rgba 1 0.2694 0.1226 1>
    Lifetime=0.24
    Mesh Scale=<6.8 14000 6.8>
    Mesh Scale Mode=::NewEnumerator4
    Position Offset=<0 0 400>
    UsePositionOffset=true
    Write Material Random=true
   )
   &InitialMeshOrientation(Mesh Axis To Orient=<0 0 1>, Mesh Orientation Mode=::NewEnumerator2, Use Rotation Vector=false)
   &DynamicMaterialParameters(
    Index 0 Param 1=1
    Index 0 Param 2=1
    Index 0 Param 3=1
    Index 0 Param 4=1
    Module.Index0Float0Write=true
    Module.Index0Float1Write=true
    Module.Index0Float2Write=true
    Module.Index0Float3Write=true
   )
  ]
  particleUpdateScript [
   &ParticleState(Kill Particles When Lifetime Has Elapsed=false, Let Infinitely Lived Particles Die When Emitter Deactivates=true)
   &ScaleMeshSize(Scale Factor=&MakeVector(
     X=1
     Y=&FloatFromCurve(
      FloatCurve=DataInterface() {propertyValues {Curve {keys [<at 0 1>,<at 0.0828 0.8388>,<at 0.1062 1.016>,<at 0.1764 0.8373>,<at 0.2189 1.019>,<at 0.2701 0.8166>,<at 0.315 0.9866>,<at 0.3779 0.8023>,<at 0.4017 0.9866>,<at 0.4571 0.8309>,<at 0.4916 0.947>,<at 0.5443 0.8141>,<at 0.5628 0.9641>,<at 0.6235 0.8065>,<at 0.6372 0.9422>,<at 0.6642 0.8484>,<at 0.7008 0.9715>,<at 0.7015 0.8114>,<at 0.7333 0.903>,<at 0.7583 0.7922>,<at 0.7673 0.978>,<at 0.8266 0.7936>,<at 0.8519 0.9406>,<at 0.8951 0.8022>,<at 0.9036 0.9437>,<at 0.9393 0.8996>,<at 0.9656 0.9866>,<at 1 1>]}}}
      CurveIndex=$System.NormalizedLoopAge
      Scale Curve=1.5
     )
     Z=1
    ))
   &ScaleMeshSize(Scale Factor=&MakeVector(X=$User.Length, Y=&Multiply_Float(A=0.36, B=$User.Intensity), Z=1)) {module=ScaleMeshSize001}
   &ScaleColor(
    Scale RGB=&VectorFromFloat(Value=&Multiply_Float(A=2, B=$User.Core_Intensity))
    ScaleA=true
    ScaleRGB=true
   ) {module=ScaleColor001}
   &SystemLocation(MaskSpawnGroup=false, Offset=&MakeVector(X=0, Y=0, Z=$User.Core_Beam_Z_Offset))
  ]
  rendererClasses [&NiagaraMeshRendererProperties]
  renderers [{rendererClass=&NiagaraMeshRendererProperties, rendererIndex=0}]
 }
 Beam_Glow {
  emitterUpdateScript [
   &EmitterState(Life Cycle Mode=::NewEnumerator1, Loop Behavior=::NewEnumerator1, Loop Duration Mode=::NewEnumerator1, UseLoopDelay=false)
   &SpawnBurst_Instantaneous(Use Spawn Probability=false, UseLoopCountLimit=false)
  ]
  particleSpawnScript [
   &InitializeParticle(
    Color=<rgba 1 0.2018 0.058 1>
    Lifetime=0.2
    Mesh Scale=<36 22 22>
    Mesh Scale Mode=::NewEnumerator4
    Position Offset=<0 0 400>
    UsePositionOffset=true
    Write Material Random=true
   )
   &InitialMeshOrientation(Mesh Axis To Orient=<0 0 1>, Mesh Orientation Mode=::NewEnumerator2, Use Rotation Vector=false)
  ]
  particleUpdateScript [
   &ParticleState(Kill Particles When Lifetime Has Elapsed=false, Let Infinitely Lived Particles Die When Emitter Deactivates=true)
   &ScaleMeshSize(Scale Factor=&MakeVector(
     X=$User.Length
     Y=&Multiply_Float(A=0.65, B=$User.Intensity)
     Z=&Multiply_Float(A=0.65, B=$User.Intensity)
    )) {module=ScaleMeshSize001}
   &ScaleMeshSize(Scale Factor=&MakeVector(
     X=1
     Y=&FloatFromCurve(
      FloatCurve=DataInterface() {propertyValues {Curve {keys [<at 0 1>,<at 0.0153 0.945>,<at 0.0515 0.995>,<at 0.0782 0.9425>,<at 0.1062 0.995>,<at 0.1404 0.9378>,<at 0.202 0.9993>,<at 0.2147 0.9425>,<at 0.2479 0.9906>,<at 0.2613 0.9342>,<at 0.315 0.9866>,<at 0.324 0.9381>,<at 0.3818 0.9963>,<at 0.4156 0.9395>,<at 0.4546 0.99>,<at 0.4792 0.9364>,<at 0.5305 0.9932>,<at 0.5513 0.9336>,<at 0.5989 0.9935>,<at 0.6451 0.9293>,<at 0.6844 0.9924>,<at 0.7069 0.9278>,<at 0.7504 0.9981>,<at 0.7608 0.9232>,<at 0.7862 0.9954>,<at 0.8069 0.9272>,<at 0.8396 0.9957>,<at 0.867 0.9177>,<at 0.8715 0.9928>,<at 0.9053 0.9372>,<at 0.9264 1.0028>,<at 0.9581 0.9341>,<at 1 1>]}}}
      CurveIndex=$System.NormalizedLoopAge
     )
     Z=&FloatFromCurve(
      FloatCurve=DataInterface() {propertyValues {Curve {keys [<at 0 1>,<at 0.0153 0.945>,<at 0.0515 0.995>,<at 0.0782 0.9425>,<at 0.1062 0.995>,<at 0.1404 0.9378>,<at 0.202 0.9993>,<at 0.2147 0.9425>,<at 0.2479 0.9906>,<at 0.2613 0.9342>,<at 0.315 0.9866>,<at 0.324 0.9381>,<at 0.3818 0.9963>,<at 0.4156 0.9395>,<at 0.4546 0.99>,<at 0.4792 0.9364>,<at 0.5305 0.9932>,<at 0.5513 0.9336>,<at 0.5989 0.9935>,<at 0.6451 0.9293>,<at 0.6844 0.9924>,<at 0.7069 0.9278>,<at 0.7504 0.9981>,<at 0.7608 0.9232>,<at 0.7862 0.9954>,<at 0.8069 0.9272>,<at 0.8396 0.9957>,<at 0.867 0.9177>,<at 0.8715 0.9928>,<at 0.9053 0.9372>,<at 0.9264 1.0028>,<at 0.9581 0.9341>,<at 1 1>]}}}
      CurveIndex=$System.NormalizedLoopAge
     )
    ))
   &ScaleColor(Scale RGB=&VectorFromFloat(Value=$User.Intensity), ScaleA=true, ScaleRGB=true)
  ]
  rendererClasses [&NiagaraMeshRendererProperties]
  renderers [{rendererClass=&NiagaraMeshRendererProperties, rendererIndex=0}]
 }
 CoreBleedLight {
  emitterUpdateScript [
   &EmitterState(
    Inactive Response=::NewEnumerator1
    Life Cycle Mode=::NewEnumerator1
    Loop Behavior=::NewEnumerator1
    Loop Delay=1.7
    Loop Duration Mode=::NewEnumerator1
    UseLoopDelay=false
   )
   &SpawnBurst_Instantaneous(Use Spawn Probability=false, UseLoopCountLimit=false)
  ]
  particleSpawnScript [
   &InitializeParticle(
    Color=<rgba 2 0.4428 0.058 1>
    Lifetime=0.4
    Mesh Scale=<4.5 4.5 8>
    Mesh Scale Mode=::NewEnumerator4
    Sprite Size Mode=::NewEnumerator3
    Uniform Sprite Size=1000
    UsePositionOffset=false
    Write Material Random=true
   )
   &InitialMeshOrientation(Mesh Orientation Mode=::NewEnumerator2, Orientation Vector=<0 0 1>, Use Rotation Vector=false)
   &SpriteFacingAndAlignment(Sprite Facing=<0 0 1>, WriteAlignment=false, WriteFacing=true)
  ]
  particleUpdateScript [
   &ParticleState(Kill Particles When Lifetime Has Elapsed=false, Let Infinitely Lived Particles Die When Emitter Deactivates=true)
   &ScaleMeshSize(Scale Factor=<2 2 -2>)
   &ScaleMeshSize(Scale Factor=&MakeVector(
     X=1
     Y=1
     Z=&FloatFromCurve(
      FloatCurve=DataInterface() {propertyValues {Curve {keys [<at 0 1>,<at 0.0153 0.8606>,<at 0.0515 0.995>,<at 0.0782 0.8582>,<at 0.1062 0.995>,<at 0.1404 0.8534>,<at 0.202 0.9993>,<at 0.2147 0.8581>,<at 0.2479 0.9906>,<at 0.2613 0.8498>,<at 0.315 0.9866>,<at 0.324 0.8537>,<at 0.3818 0.9963>,<at 0.4156 0.8552>,<at 0.4546 0.99>,<at 0.4792 0.852>,<at 0.5305 0.9932>,<at 0.5513 0.8492>,<at 0.5989 0.9935>,<at 0.6451 0.845>,<at 0.6844 0.9924>,<at 0.7069 0.8434>,<at 0.7504 0.9981>,<at 0.7608 0.8389>,<at 0.7862 0.9954>,<at 0.8069 0.8428>,<at 0.8396 0.9957>,<at 0.867 0.8333>,<at 0.8715 0.9928>,<at 0.9053 0.8528>,<at 0.9264 1.0028>,<at 0.9581 0.8497>,<at 1 1>]}}}
      CurveIndex=$System.NormalizedLoopAge
     )
    )) {module=ScaleMeshSize001}
   &ScaleColor(Scale Alpha=$User.Intensity, ScaleA=true, ScaleRGB=true)
   /SetVariables(Particles.Position=&AddVectorToPosition(Position=<0 0 0>, Vector=$System.OrbWPOLocal)) {isSetParameters=true, module=SetVariables_#3}
   /SetVariables(Particles.HorizontalFlare_Color:LinearColor=&Multiply_LinearColorByFloat(
     Linear Color=&Multiply_LinearColorByFloat(
      Linear Color=<rgba 0.5 0.0193 0.0114 1>
      Float=&MakeFloatFromVector4(
       Vector 4=&MakeFloat4FromVectorAndFloat(
        "Vector (RGB)"=&Multiply_VectorByFloat(
         Vector=&VectorFromFloat(Value=&Waveform(Global Input Position=1, "Waveform [1]"=::NewEnumerator12, "[1] Frequency"=10, "[1] Amplitude Min/Max"=<0.85 1>))
         Float=1.15
        )
        "Float (Alpha)"=1
       )
       Channel=::NewEnumerator0
      )
     )
     Float=1
    )) {isSetParameters=true}
   /SetVariables(
    Particles.FlatFlareDyn:Vector4f=&MakeVector4(
     X=0
     Y=&Waveform("Waveform [1]"=::NewEnumerator12, "[1] Frequency"=20, "[1] Amplitude Min/Max"=<0 0.2>)
     Z=0
     W=0
    )
    Particles.FlatFlareSize:Vector2f=&Multiply_Vector2D_ByFloat(
     Float=$User.Intensity
     Vector2D=&Multiply_Vector2(
      A=&Multiply_Vector2D(A=<-700 100>, B=<1 1>)
      B=&Vector2DFromFloat(Value=8)
     )
    )
   ) {isSetParameters=true}
  ]
  rendererClasses [&NiagaraMeshRendererProperties, &NiagaraSpriteRendererProperties]
  renderers [
   {rendererClass=&NiagaraMeshRendererProperties, rendererIndex=0}
   {rendererClass=&NiagaraSpriteRendererProperties, rendererIndex=1}
   {rendererClass=&NiagaraSpriteRendererProperties, rendererIndex=2}
  ]
 }
 DarkSPline001 {
  emitterUpdateScript [
   &EmitterState(
    Enable Distance Culling=true
    Life Cycle Mode=::NewEnumerator1
    Loop Delay=0.8
    Loop Duration=3
    Max Distance Response=::Awaken
    MaxDistance=32000
    MinDistance=4000
    Module.bApplySpawnCountScaleByDistance=true
    Module.bUseMaxDistance=true
    Module.bUseMinDistance=false
    Scalability Mode=::NewEnumerator1
    Scale Spawn Count=true
    UseLoopDelay=false
   )
   &SpawnRate(SpawnRate=&Multiply_Float(A=10, B=$User.Intensity), Use Spawn Probability=false)
  ]
  particleSpawnScript [
   &InitializeParticle(
    Color=<rgba 5 1.1728 0 0.75>
    Lifetime Max=1.5
    Lifetime Mode=::NewEnumerator1
    Mass Max=0.2
    Mass Min=0
    Mass Mode=::NewEnumerator3
    Mesh Scale Max=<15 15 10>
    Mesh Scale Min=<-15 -15 5>
    Mesh Scale Mode=::NewEnumerator2
    Position Offset=<0 0 -500>
    Sprite Rotation Mode=::NewEnumerator2
    UsePositionOffset=false
    Write Material Random=true
   )
   &InitialMeshOrientation(
    Mesh Orientation Mode=::NewEnumerator2
    Rotation=&RandomRangeVector(Minimum=<0 0 0>, Maximum=<0 0 1>)
    Use Rotation Vector=true
   )
   &AddVelocity(
    Cone Angle=1
    Cone Axis=<0 0 -1>
    Module.UseConeFalloff=false
    Module.UseVelDistribution=false
    Override Seed=false
    Velocity Mode=::NewEnumerator2
    Velocity Speed=&RandomRangeFloat(Minimum=200, Maximum=1500)
   )
   &ShapeLocation(
    Cylinder Height Midpoint=1
    Cylinder Radius=&Multiply_Float(A=60, B=$User.Intensity)
    MaskSpawnGroup=false
    Override Seed=false
    Shape Primitive=::NewEnumerator1
    Surface Only=false
    UseOwnerScale=true
   ) {module=ShapeLocation001}
  ]
  particleUpdateScript [
   &ParticleState()
   &ScaleMeshSize(Scale Factor=&VectorFromCurve(
     VectorCurve=DataInterface() {propertyValues {
       XCurve {keys [
         {arriveTangent=1.9584, interpMode=RCIM_Cubic, leaveTangent=1.9584, tangentMode=RCTM_User, time=0, value=0.8}
         {arriveTangent=0.0224, interpMode=RCIM_Cubic, leaveTangent=0.0224, tangentMode=RCTM_User, time=0.9, value=2}
        ]}
       YCurve {keys [
         {arriveTangent=1.9584, interpMode=RCIM_Cubic, leaveTangent=1.9584, tangentMode=RCTM_User, time=0, value=1}
         {arriveTangent=0.0224, interpMode=RCIM_Cubic, leaveTangent=0.0224, tangentMode=RCTM_User, time=0.9, value=2}
        ]}
       ZCurve {keys [
         {arriveTangent=1.9584, interpMode=RCIM_Cubic, leaveTangent=1.9584, tangentMode=RCTM_User, time=0, value=0.4}
         {arriveTangent=1.1179, interpMode=RCIM_Cubic, leaveTangent=1.1179, tangentMode=RCTM_User, time=0.3, value=0.75}
         {arriveTangent=0.0224, interpMode=RCIM_Cubic, leaveTangent=0.0224, tangentMode=RCTM_User, time=1, value=1}
        ]}
      }}
     Scale Curve=&Multiply_VectorByFloat(
      Vector=&Multiply_VectorByFloat(Vector=<10 10 50>, Float=1)
      Float=&RandomRangeFloat(Minimum=0.07, Maximum=0.08)
     )
    ))
   &Drag(Drag=&RandomRangeFloat(Minimum=0.25, Maximum=0.35), Use Linear Drag=true, Use Rotational Drag=false)
   &ScaleColor(
    Scale Alpha=&FloatFromCurve(
     FloatCurve=DataInterface() {propertyValues {Curve {keys [
         {arriveTangent=0.1304, interpMode=RCIM_Cubic, leaveTangent=0.1304, tangentMode=RCTM_User, time=0, value=0}
         {interpMode=RCIM_Cubic, tangentMode=RCTM_User, time=0.1, value=1}
         {tangentMode=RCTM_User, time=1, value=0}
        ]}}}
     Scale Curve=0.5
    )
    ScaleA=true
    ScaleRGB=true
   )
   &CameraOffset(Camera Offset Amount=&Multiply_Float(
     A=/Game/Effects/Niagara/Modules/Math/FN_CameraDistance(
      0Dist:NiagaraFloat=0
      1Dist:NiagaraFloat=100
      Camera Query:NiagaraDataInterfaceCamera=DataInterface() {propertyValues {bRequireCurrentFrameData=false}}
      Position:NiagaraPosition=$Particles.Position
     )
     B=300
    ))
   &SolveForcesAndVelocity(Clamp Velocity=false, Limit Acceleration=false)
   &DynamicMaterialParameters(
    Index 0 Param 1=0.5
    Index 0 Param 3=1
    Module.Index0Float0Write=true
    Module.Index0Float1Write=true
    Module.Index0Float2Write=true
    Module.Index0Float3Write=true
    Module.Index1Float0Write=true
    Module.Index1Float1Write=true
    Module.Index1Float2Write=false
    Module.Index1Float3Write=false
    Write Parameter Index 1=::NewEnumerator1
   )
  ]
  rendererClasses [&NiagaraMeshRendererProperties]
  renderers [{rendererClass=&NiagaraMeshRendererProperties, rendererIndex=0}]
 }
 EnergyLeak001 {
  emitterUpdateScript [
   &EmitterState(Life Cycle Mode=::NewEnumerator1, Loop Delay=0.4, Loop Duration=2, UseLoopDelay=false)
   &SpawnBurst_Instantaneous(Use Spawn Probability=false, UseLoopCountLimit=false)
  ]
  particleSpawnScript [
   &InitializeParticle(
    Color=<rgba 0.9271 0.1288 0 1>
    Lifetime=4
    Mesh Scale Mode=::NewEnumerator3
    Mesh Uniform Scale=&Multiply_Float(A=10, B=1)
    Sprite Rotation Mode=::NewEnumerator2
    Sprite Size Mode=::NewEnumerator3
    Sprite UV Mode=::NewEnumerator3
    Uniform Sprite Size=&Multiply_Float(A=900, B=1)
    UsePositionOffset=false
    Write Material Random=true
   )
   &InitialMeshOrientation(Mesh Axis To Orient=<0 0 1>, Mesh Orientation Mode=::NewEnumerator2, Use Rotation Vector=false)
  ]
  particleUpdateScript [
   &ParticleState()
   &ScaleSpriteSize(
    Uniform Curve Scale=2
    Uniform Curve Sprite Scale=DataInterface() {propertyValues {Curve {keys [
        {interpMode=RCIM_Cubic, tangentMode=RCTM_Break, time=0, value=0.5}
        {arriveTangent=-0.3016, interpMode=RCIM_Cubic, tangentMode=RCTM_Break, time=0.75, value=1}
       ]}}}
   )
   &Timeline(
    Play Rate=&FloatFromCurve(FloatCurve=DataInterface() {propertyValues {Curve {keys [<at 0 2>,<at 1 1.3>]}}})
    Timeline End=1000
   )
   &ScaleColor(
    Scale Alpha=&FloatFromCurve(
     FloatCurve=DataInterface() {propertyValues {Curve {keys [
         {arriveTangent=9.0725, interpMode=RCIM_Cubic, leaveTangent=9.0725, tangentMode=RCTM_User, time=0, value=0}
         {arriveTangent=0.1216, interpMode=RCIM_Cubic, leaveTangent=0.1216, tangentMode=RCTM_User, time=0.2, value=0.9847}
         {arriveTangent=-3.3987, interpMode=RCIM_Cubic, leaveTangent=-3.3987, tangentMode=RCTM_User, time=1, value=0}
        ]}}}
     Scale Curve=&Multiply_Float(A=1.2, B=$User.Intensity)
    )
    Scale RGB=&VectorFromFloat(Value=1)
    ScaleA=true
    ScaleRGB=true
   ) {module=ScaleColor002}
   &DynamicMaterialParameters(
    Index 0 Param 1=&Multiply_Float(A=$Particles.Timeline.Time, B=2)
    Module.Index0Float0Write=true
    Module.Index0Float1Write=true
    Module.Index0Float2Write=true
    Module.Index0Float3Write=true
   )
   /SetVariables(Particles.Position=&AddVectorToPosition(Position=<0 0 0>, Vector=$System.OrbWPOLocal)) {isSetParameters=true}
   &CameraOffset(Camera Offset Amount=&Multiply_Float(
     A=&FN_CameraDistance(
      0Dist:NiagaraFloat=0
      1Dist:NiagaraFloat=800
      Camera Query:NiagaraDataInterfaceCamera=DataInterface() {propertyValues {bRequireCurrentFrameData=false}}
      Position:NiagaraPosition=$Particles.Position
     )
     B=1000
    )) {module=CameraOffset001}
  ]
  rendererClasses [&NiagaraSpriteRendererProperties]
  renderers [{rendererClass=&NiagaraSpriteRendererProperties}]
 }
 Impact_Flare {
  simTarget=GPUComputeSim
  emitterUpdateScript [
   &EmitterState(Life Cycle Mode=::NewEnumerator1, Loop Behavior=::NewEnumerator1, Loop Duration Mode=::NewEnumerator1, UseLoopDelay=false)
   &SpawnBurst_Instantaneous(Use Spawn Probability=false, UseLoopCountLimit=false)
  ]
  particleSpawnScript [
   &InitializeParticle(
    Color=<rgba 5 0.9743 0.139 1>
    Lifetime=2
    Sprite Size Mode=::NewEnumerator3
    Uniform Sprite Size=2000
    UsePositionOffset=false
    Write Material Random=true
   )
   &DynamicMaterialParameters(
    Index 0 Param 1=1
    Index 0 Param 2=1
    Index 0 Param 3=1
    Index 0 Param 4=1
    Module.Index0Float0Write=true
    Module.Index0Float1Write=true
    Module.Index0Float2Write=true
    Module.Index0Float3Write=true
   )
  ]
  particleUpdateScript [
   &ParticleState(Kill Particles When Lifetime Has Elapsed=false, Let Infinitely Lived Particles Die When Emitter Deactivates=true)
   &ScaleSpriteSize(Scale Sprite Size Mode=::NewEnumerator0, Uniform Scale Factor=$User.Intensity)
   &CameraOffset(Camera Offset Amount=300)
   &FN_Sprite_Attributes(
    Color=&Multiply_LinearColorByFloat(
     Linear Color=$Particles.Color
     Float=&Multiply_Float(
      A=&MakeFloatFromVector4(
       Vector 4=&MakeFloat4FromVectorAndFloat(
        "Vector (RGB)"=&Multiply_VectorByFloat(
         Vector=&VectorFromFloat(Value=&Waveform(Global Input Position=1, "Waveform [1]"=::NewEnumerator12, "[1] Frequency"=20, "[1] Amplitude Min/Max"=<0.6 1>))
         Float=2
        )
        "Float (Alpha)"=1
       )
       Channel=::NewEnumerator0
      )
      B=&FloatFromCurve(FloatCurve=DataInterface() {propertyValues {Curve {keys [<at 0 0>,<at 1 1>]}}})
     )
    )
    Create Color Attribute=true
    Create SpriteSize Attribute=true
    SpriteSize=&Multiply_Vector2D(A=$Particles.SpriteSize, B=<1 0.4232>)
    Write to Module Namespace=true
   )
   &ScaleColor(
    Scale Alpha=&Multiply_Float(A=$User.Intensity, B=$User.UpLight_ExtraIntensity_Mult)
    Scale RGB=&VectorFromFloat(Value=&MakeFloatFromVector4(
      Vector 4=&MakeFloat4FromVectorAndFloat(
       "Vector (RGB)"=&Multiply_VectorByFloat(
        Vector=&VectorFromFloat(Value=&Waveform(Global Input Position=1, "Waveform [1]"=::NewEnumerator12, "[1] Frequency"=30, "[1] Amplitude Min/Max"=<0.8 1>))
        Float=1.2
       )
       "Float (Alpha)"=1
      )
      Channel=::NewEnumerator0
     ))
    ScaleA=true
    ScaleRGB=true
   )
   /SetVariables(Particles.Position=&AddVectorToPosition(Position=<0 0 0>, Vector=$System.OrbWPOLocal)) {isSetParameters=true, module=SetVariables_#3}
  ]
  rendererClasses [&NiagaraSpriteRendererProperties]
  renderers [
   {rendererClass=&NiagaraSpriteRendererProperties}
   {rendererClass=&NiagaraSpriteRendererProperties, rendererIndex=1}
  ]
 }
 Light_Up {
  emitterUpdateScript [
   &EmitterState(Life Cycle Mode=::NewEnumerator1, Loop Behavior=::NewEnumerator1, Loop Duration Mode=::NewEnumerator1, UseLoopDelay=false)
   &SpawnBurst_Instantaneous(Use Spawn Probability=false, UseLoopCountLimit=false)
  ]
  particleSpawnScript [&InitializeParticle(
    Color=<rgba 200 27.9666 0 1>
    Lifetime=2.2
    Position=<0 0 -100>
    Position Mode=::NewEnumerator1
    UsePositionOffset=false
    Write Material Random=true
   )]
  particleUpdateScript [
   &ParticleState(Kill Particles When Lifetime Has Elapsed=false, Let Infinitely Lived Particles Die When Emitter Deactivates=true)
   /SetVariables(
    Particles.LightExponent=2
    Particles.LightRadius=&Multiply_Float(A=1500, B=&Multiply_Float(A=$User.Intensity, B=$User.UpLight_ExtraIntensity_Mult))
   ) {isSetParameters=true}
   /SetVariables(Particles.Position=&AddVectorToPosition(Position=<0 0 0>, Vector=$System.OrbWPOLocal)) {isSetParameters=true, module=SetVariables_#3}
   &ScaleColor(
    Scale RGB=&Multiply_VectorByFloat(
     Vector=&VectorFromFloat(Value=&MakeFloatFromVector4(
       Vector 4=&MakeFloat4FromVectorAndFloat(
        "Vector (RGB)"=&Multiply_VectorByFloat(
         Vector=&VectorFromFloat(Value=&Waveform(Global Input Position=1, "Waveform [1]"=::NewEnumerator12, "[1] Frequency"=30, "[1] Amplitude Min/Max"=<0.7 1>))
         Float=1
        )
        "Float (Alpha)"=1
       )
       Channel=::NewEnumerator0
      ))
     Float=&Multiply_Float(A=$User.Intensity, B=$User.UpLight_ExtraIntensity_Mult)
    )
    ScaleA=true
    ScaleRGB=true
   ) {module=ScaleColor001}
  ]
  rendererClasses [&NiagaraLightRendererProperties]
  renderers [{rendererClass=&NiagaraLightRendererProperties, rendererIndex=0}]
 }
 Lightning_V01001 {
  emitterUpdateScript [
   &EmitterState(
    Enable Distance Culling=true
    Max Distance Response=::Awaken
    MaxDistance=32000
    MinDistance=4000
    Module.bApplySpawnCountScaleByDistance=true
    Module.bUseMaxDistance=true
    Module.bUseMinDistance=false
    Scalability Mode=::NewEnumerator1
    Scale Spawn Count=true
   )
   &SpawnRate(SpawnRate=10, Use Spawn Probability=false)
  ]
  particleSpawnScript [
   &PartitionParticles(Number Of Partitions=3)
   &InitializeParticle(
    Color=&RandomRangeLinearColor(Minimum=<rgba 2 0.5795 0.0104 1>, Maximum=<rgba 1 0.043 0.0468 1>)
    Lifetime=&RandomRangeFloat(Minimum=1, Maximum=1.5)
    Mesh Renderer Array Visibility Mode=::NewEnumerator1
    Mesh Renderer Info=DataInterface() {propertyValues {MeshRenderer=&self:Lightning_V01001_0.NiagaraMeshRendererProperties_0}}
    Mesh Renderer Mesh Index:NiagaraInt32=$Particles.PartitionParticles.Partition
    Mesh Scale=&MakeVector(
     X=&Multiply_Float(A=$User.Length, B=&RandomRangeFloat(Minimum=4, Maximum=4.5))
     Y=&Multiply_Float(A=$User.Intensity, B=&RandomRangeFloat(Minimum=1, Maximum=1.2))
     Z=&Multiply_Float(A=$User.Intensity, B=&RandomRangeFloat(Minimum=1, Maximum=1.2))
    )
    Mesh Scale Mode=::NewEnumerator4
    UsePositionOffset=false
    Write Material Random=true
   )
   &InitialMeshOrientation(
    Mesh Axis To Orient=<0 0 -1>
    Mesh Orientation Mode=::NewEnumerator2
    Rotation=&RandomRangeVector(Minimum=<-1 0 0>, Maximum=<1 0 0>)
    Use Rotation Vector=true
   )
   &ShapeLocation(
    MaskSpawnGroup=false
    Non Uniform Scale=<1 1 0.5>
    Offset=&AddVectorToPosition(Position=&MakeVector(X=0, Y=0, Z=$User.Lightning_Z_Offset), Vector=$System.OrbWPOLocal)
    Offset Mode=::NewEnumerator0
    Override Seed=false
    Sphere Radius=&Multiply_Float(A=10, B=$User.Intensity)
    Sphere Surface Distribution=0.6548
    UseOwnerScale=true
   )
  ]
  particleUpdateScript [
   &ParticleState()
   &ScaleMeshSize(Scale Factor=&VectorFromCurve(VectorCurve=DataInterface() {propertyValues {
       XCurve {keys [{interpMode=RCIM_Cubic, leaveTangent=0.0038, tangentMode=RCTM_Break, time=0, value=1}]}
       YCurve {keys [{interpMode=RCIM_Cubic, tangentMode=RCTM_Break, time=0, value=1.2}, {interpMode=RCIM_Cubic, tangentMode=RCTM_Break, time=1, value=1}]}
       ZCurve {keys [{interpMode=RCIM_Cubic, tangentMode=RCTM_Break, time=0, value=1.2}, {interpMode=RCIM_Cubic, tangentMode=RCTM_Break, time=1, value=1}]}
      }}))
   /SetVariables(
    Particles.Flare:NiagaraFloat=&Lerp_Float(
     A=-0.4
     B=&FloatFromCurve(
      FloatCurve=DataInterface() {propertyValues {Curve {keys [
          {arriveTangent=-9.9823, interpMode=RCIM_Cubic, time=0.0073, value=1}
          {arriveTangentWeight=0.1021, interpMode=RCIM_Cubic, leaveTangentWeight=0.2548, tangentMode=RCTM_User, time=0.35, value=0}
          {arriveTangentWeight=1.9969, interpMode=RCIM_Cubic, leaveTangent=19.1881, leaveTangentWeight=0.804, tangentMode=RCTM_Break, time=0.5, value=0}
          {arriveTangent=-4.9513, interpMode=RCIM_Cubic, leaveTangent=-4.9513, tangentMode=RCTM_User, time=0.6, value=1}
          <at 0.9 0>
         ]}}}
      Scale Curve=0.1
     )
     Alpha=&FloatFromCurve(FloatCurve=DataInterface() {propertyValues {Curve {keys [<at 0 1>,<at 1 0>]}}})
    )
    Particles.Flicker:NiagaraFloat=&Waveform("Waveform [1]"=::NewEnumerator0, "[1] Frequency"=20, "[1] Amplitude Min/Max"=<0.5 1>)
   ) {isSetParameters=true}
   &UpdateMeshOrientation(Rotation Rate=&RandomRangeFloat(Minimum=0.1, Maximum=0.2))
   &DynamicMaterialParameters(
    Index 0 Param 1=&FloatFromCurve(
     FloatCurve=DataInterface() {propertyValues {Curve {keys [{interpMode=RCIM_Cubic, tangentMode=RCTM_Break, time=0, value=1}, {interpMode=RCIM_Cubic, tangentMode=RCTM_Break, time=0.8, value=0}]}}}
     Scale Curve=3
    )
    Index 0 Param 2=&FloatFromCurve(
     FloatCurve=DataInterface() {propertyValues {Curve {keys [
         {interpMode=RCIM_Cubic, leaveTangent=-10.0436, tangentMode=RCTM_Break, time=0.1682, value=2}
         {arriveTangent=-0.9209, interpMode=RCIM_Cubic, leaveTangent=-0.9209, tangentMode=RCTM_User, time=0.296, value=1.3359}
         {interpMode=RCIM_Cubic, tangentMode=RCTM_Break, time=1, value=1}
        ]}}}
     Scale Curve=&Multiply_Float(A=1, B=&RandomRangeFloat(Minimum=3, Maximum=5))
    )
    Index 0 Param 3=&Lerp_Float(A=0.1, B=0.2, Alpha=$Particles.Flare)
    Index 0 Param 4=$Particles.MaterialRandom
    Index 1 Param 1=&FloatFromCurve(FloatCurve=DataInterface() {propertyValues {Curve {defaultValue=1, keys [{tangentMode=RCTM_User, time=0, value=0}, {tangentMode=RCTM_User, time=1, value=1}]}}})
    Index 1 Param 2=&RandomRangeFloat(Minimum=3, Maximum=5)
    Index 1 Param 3=1
    Module.Index0Float0Write=true
    Module.Index0Float1Write=true
    Module.Index0Float2Write=true
    Module.Index0Float3Write=true
    Module.Index1Float0Write=true
    Module.Index1Float1Write=true
    Module.Index1Float2Write=true
    Module.Index1Float3Write=false
    Write Parameter Index 1=::NewEnumerator1
   )
   &ScaleColor(
    Scale RGB=&VectorFromFloat(Value=&Multiply_Float(A=$Particles.Flicker, B=&Lerp_Float(A=2, B=5, Alpha=$Particles.Flare)))
    ScaleA=true
    ScaleRGB=true
   )
   &ScaleColor(Scale RGB=&VectorFromFloat(Value=$User.Intensity), ScaleA=true, ScaleRGB=true) {module=ScaleColor001}
  ]
  rendererClasses [&NiagaraMeshRendererProperties]
  renderers [{rendererClass=&NiagaraMeshRendererProperties, rendererIndex=0}]
 }
 Ray {
  emitterUpdateScript [
   &EmitterState(
    Enable Distance Culling=true
    Max Distance Response=::Awaken
    MaxDistance=22000
    MinDistance=4000
    Module.bApplySpawnCountScaleByDistance=true
    Module.bUseMaxDistance=true
    Module.bUseMinDistance=false
    Scalability Mode=::NewEnumerator1
    Scale Spawn Count=true
   )
   &SpawnRate(SpawnRate=&Multiply_Float(A=20, B=$User.Intensity), Use Spawn Probability=false)
  ]
  particleSpawnScript [
   &InitializeParticle(
    Color=<rgba 8 1.4033 0 1>
    Lifetime=&Multiply_Float(A=1, B=&RandomRangeFloat(Minimum=0.22, Maximum=0.25))
    Sprite Size Max=<150 1200>
    Sprite Size Min=<140 1000>
    Sprite Size Mode=::NewEnumerator2
    Sprite UV Mode=::NewEnumerator4
    Sprite UV Scale:Vector2f=<1 -1>
    UsePositionOffset=false
    Write Material Random=true
   )
   &DynamicMaterialParameters(
    Index 0 Param 1=&RandomRangeFloat(Minimum=1.5, Maximum=2)
    Index 0 Param 2=2
    Index 0 Param 3=&RandomRangeFloat(Minimum=1.5, Maximum=2)
    Index 0 Param 4=2
    Module.Index0Float0Write=true
    Module.Index0Float1Write=true
    Module.Index0Float2Write=true
    Module.Index0Float3Write=true
   )
   &ShapeLocation(
    Cylinder Height=500
    Cylinder Radius=120
    MaskSpawnGroup=false
    Offset=&AddVectorToPosition(Position=<0 0 0>, Vector=$System.OrbWPOLocal)
    Offset Mode=::NewEnumerator0
    Override Seed=false
    Shape Primitive=::NewEnumerator1
    Surface Only=true
    Surface Only Band Thickness=0.3
    Use Endcaps In Surface Only Mode=false
    UseOwnerScale=true
   )
   &AddVelocity(Velocity=&Multiply_VectorByFloat(Vector=&RandomRangeVector(Minimum=<0 0 -11000>, Maximum=<0 1 -12000>), Float=1))
  ]
  particleUpdateScript [
   &ParticleState()
   &SolveForcesAndVelocity(Clamp Velocity=false, Limit Acceleration=false)
   &ScaleColor(
    Scale Alpha=&FloatFromCurve(FloatCurve=DataInterface() {propertyValues {Curve {keys [{interpMode=RCIM_Cubic, tangentMode=RCTM_Break, time=0.8, value=1}, {interpMode=RCIM_Cubic, tangentMode=RCTM_Break, time=1, value=0}]}}})
    Scale RGB=&VectorFromFloat(Value=$User.Intensity)
    ScaleA=true
    ScaleRGB=true
   )
  ]
  rendererClasses [&NiagaraSpriteRendererProperties]
  renderers [{rendererClass=&NiagaraSpriteRendererProperties}]
 }
 SmallLightning_Flare {
  emitterUpdateScript [
   &EmitterState(Life Cycle Mode=::NewEnumerator1, Loop Behavior=::NewEnumerator1, Loop Duration Mode=::NewEnumerator1, UseLoopDelay=false)
   &SpawnBurst_Instantaneous(
    Spawn Count=&MakeCustomIntFromBool(Bool=$User.SpawnSmallLightningLFare, True INT=5, False INT=0)
    Use Spawn Probability=false
    UseLoopCountLimit=false
   )
  ]
  particleSpawnScript [
   &InitializeParticle(
    Color=<rgba 1 0.0956 0 1>
    Lifetime=2
    Position Mode=::NewEnumerator1
    Sprite Rotation Mode=::NewEnumerator2
    Sprite Size Mode=::NewEnumerator3
    Uniform Sprite Size=900
    UsePositionOffset=false
    Write Material Random=true
   )
   /SetVariables(Particles.ExecutionIndex:NiagaraFloat=&ReturnNormalizedExecIndex(Normalized Index Includes One=false)) {isSetParameters=true}
   &DynamicMaterialParameters(
    Index 0 Param 1=1
    Index 0 Param 2=1
    Index 0 Param 3=1
    Index 0 Param 4=1
    Module.Index0Float0Write=true
    Module.Index0Float1Write=true
    Module.Index0Float2Write=true
    Module.Index0Float3Write=true
   )
  ]
  particleUpdateScript [
   &ParticleState(Kill Particles When Lifetime Has Elapsed=false, Let Infinitely Lived Particles Die When Emitter Deactivates=true)
   &ShapeLocation(
    MaskSpawnGroup=false
    Offset=&AddVectorToPosition(Position=<0 0 -79.045>, Vector=$System.OrbWPOLocal)
    Offset Mode=::NewEnumerator0
    Rotation Angle=0.081
    Rotation Axis=<0 0 1>
    Rotation Mode=::NewEnumerator1
    Sphere Distribution=::NewEnumerator1
    Sphere Radius=859.7271
    U Position=$Particles.ExecutionIndex
    UseOwnerScale=true
   )
   &FN_Sprite_Attributes(
    Color=&Multiply_LinearColorByFloat(
     Linear Color=$Particles.Color
     Float=&Multiply_Float(A=2, B=&FloatFromCurve(FloatCurve=DataInterface() {propertyValues {Curve {keys [<at 0 0>,<at 1 1>]}}}))
    )
    Create Color Attribute=true
    Create SpriteSize Attribute=true
    SpriteSize=&Multiply_Vector2D(A=$Particles.SpriteSize, B=<1 0.4232>)
    Write to Module Namespace=true
   )
   &ScaleColor(
    Scale Alpha=$User.SmallLightning_FlareIntensity
    Scale RGB=&VectorFromFloat(Value=&Waveform(
      Global Amplitude Scale=&Multiply_Float(A=1, B=2)
      "Waveform [1]"=::NewEnumerator0
      "[1] Frequency"=7
      "[1] Amplitude Min/Max"=<0.7 1>
     ))
    ScaleA=true
    ScaleRGB=true
   )
   &CameraOffset(Camera Offset Amount=120)
  ]
  rendererClasses [&NiagaraSpriteRendererProperties]
  renderers [{rendererClass=&NiagaraSpriteRendererProperties}]
 }
 UP_Lightning_Small001 {
  emitterUpdateScript [
   &EmitterState(
    Enable Distance Culling=true
    Max Distance Response=::Awaken
    MaxDistance=32000
    MinDistance=4000
    Module.bApplySpawnCountScaleByDistance=true
    Module.bUseMaxDistance=true
    Module.bUseMinDistance=false
    Scalability Mode=::NewEnumerator1
    Scale Spawn Count=true
   )
   &SpawnRate(SpawnRate=&Multiply_Float(A=12, B=$User.SmallLightning_Rate), Use Spawn Probability=false)
   &SpawnRate(SpawnGroup=1, SpawnRate=&Multiply_Float(A=12, B=$User.SmallLightning_Rate), Use Spawn Probability=false) {module=SpawnRate001}
   &SpawnRate(SpawnGroup=2, SpawnRate=&Multiply_Float(A=12, B=$User.SmallLightning_Rate), Use Spawn Probability=false) {module=SpawnRate002}
   &SpawnRate(SpawnGroup=3, SpawnRate=&Multiply_Float(A=12, B=$User.SmallLightning_Rate), Use Spawn Probability=false) {module=SpawnRate003}
   &SpawnRate(SpawnGroup=4, SpawnRate=&Multiply_Float(A=12, B=$User.SmallLightning_Rate), Use Spawn Probability=false) {module=SpawnRate004}
  ]
  particleSpawnScript [
   /SetVariables(Particles.VisibilityTag=&SpawnGroup()) {isSetParameters=true}
   &InitializeParticle(
    Color=<rgba 20 0 0.0527 1>
    Lifetime=&Multiply_Float(A=1, B=&RandomRangeFloat(Minimum=0.18, Maximum=0.22))
    Mesh Scale=&RandomRangeVector(Minimum=<12 6 6>, Maximum=<13 7 7>)
    Mesh Scale Mode=::NewEnumerator4
    UsePositionOffset=false
    Write Material Random=true
   )
   /SetVariables(Particles.Random=&RandomRangeFloat(Minimum=0, Maximum=1)) {isSetParameters=true}
   &InitialMeshOrientation(Mesh Orientation Mode=::NewEnumerator7, Rotation=<0 -0.37 0.06>, Rotation Coordinate Space=::NewEnumerator2, Use Rotation Vector=true)
   /SetVariables(Particles.Orientation_02:Quat4f=<0 -1.001 -0.9486 1>) {isSetParameters=true}
   /SetVariables(Particles.Orientation_03:Quat4f=<0 -0.45 -0.3 1>) {isSetParameters=true}
   /SetVariables(Particles.Orientation_04:Quat4f=<0 -0.43 0.22 1>) {isSetParameters=true}
   /SetVariables(Particles.Orientation_05:Quat4f=<-0.7789 -1.3832 0.2325 1>) {isSetParameters=true}
  ]
  particleUpdateScript [
   &ParticleState()
   &ScaleMeshSize(Scale Factor=&VectorFromCurve(VectorCurve=DataInterface() {propertyValues {
       XCurve {keys [
         {interpMode=RCIM_Cubic, leaveTangent=0.5336, tangentMode=RCTM_Break, time=0, value=0.7}
         {interpMode=RCIM_Cubic, tangentMode=RCTM_Break, time=0.8, value=1}
        ]}
       YCurve {keys [
         {interpMode=RCIM_Cubic, leaveTangent=1.3587, tangentMode=RCTM_Break, time=0, value=0}
         {interpMode=RCIM_Cubic, tangentMode=RCTM_Break, time=1, value=1}
        ]}
       ZCurve {keys [
         {interpMode=RCIM_Cubic, leaveTangent=1.3587, tangentMode=RCTM_Break, time=0, value=0}
         {interpMode=RCIM_Cubic, tangentMode=RCTM_Break, time=1, value=1}
        ]}
      }})) {module=ScaleMeshSize001}
   &DynamicMaterialParameters(
    Index 0 Param 1=1
    Index 0 Param 2=1
    Index 0 Param 3=&FloatFromCurve(
     FloatCurve=DataInterface() {propertyValues {Curve {keys [
         {interpMode=RCIM_Cubic, tangentMode=RCTM_Break, time=0, value=0}
         {arriveTangent=0.3846, interpMode=RCIM_Cubic, tangentMode=RCTM_Break, time=1, value=0.3}
        ]}}}
     Scale Curve=3
    )
    Index 0 Param 4=&FloatFromCurve(FloatCurve=DataInterface() {propertyValues {Curve {keys [
         {interpMode=RCIM_Cubic, leaveTangent=3.2712, tangentMode=RCTM_Break, time=0, value=1}
         {interpMode=RCIM_Cubic, tangentMode=RCTM_Break, time=1, value=3}
        ]}}})
    Index 1 Param 1=&RandomRangeFloat(Minimum=15, Maximum=20)
    Index 1 Param 2=&Add_Float(
     A=&RandomRangeFloat(Minimum=0, Maximum=10)
     B=&FloatFromCurve(FloatCurve=DataInterface() {propertyValues {Curve {keys [<at 0 0.2>,<at 1 0>]}}})
    )
    Index 1 Param 3=0.5
    Index 1 Param 4=1
    Module.Param0WriteEnabled=true
    Module.Param0WriteEnabled1=true
    Module.Param1WriteEnabled=false
    Module.Param1WriteEnabled1=true
    Module.Param2WriteEnabled=true
    Module.Param2WriteEnabled1=true
    Module.Param3WriteEnabled=true
    Module.Param3WriteEnabled1=true
    Write Parameter Index 1=true
   )
   &SolveForcesAndVelocity(Clamp Velocity=false, Limit Acceleration=false)
   /SetVariables(Particles.Lightning glow color:LinearColor=&MakeLinearColorFromVectorAndFloat(
     "Vector (RGB)"=&MakeVectorFromLinearColorRGB(Color=&MultiplyLinearColor(
       Linear Color=<rgba 10 2.9914 0 1>
       Scale Factor=&MakeVector4_FromFloat(Float=&RandomRangeFloat(Minimum=0.5, Maximum=1, Evaluation Type=::NewEnumerator1))
      ))
     "Float (Alpha)"=&Multiply_Float(
      A=&FloatFromCurve(
       FloatCurve=DataInterface() {propertyValues {Curve {keys [<at 0 1 RCIM_Cubic>, {interpMode=RCIM_Cubic, leaveTangent=-3.6307, time=0.9, value=0}]}}}
       Scale Curve=&FN_FrontEndFloatModifier(Gameplay=1, Front End=0.6)
      )
      B=7
     )
    )) {isSetParameters=true}
   &ScaleColor(
    Scale Alpha=&FloatFromCurve(FloatCurve=DataInterface() {propertyValues {Curve {keys [
         {interpMode=RCIM_Cubic, leaveTangent=3.1271, tangentMode=RCTM_Break, time=0, value=0}
         {interpMode=RCIM_Cubic, tangentMode=RCTM_Break, time=0.5, value=1}
        ]}}})
    ScaleA=true
    ScaleRGB=true
   )
   &ScaleColor(
    Scale Alpha=&Multiply_Float(A=2, B=$User.Intensity)
    Scale RGB=&VectorFromFloat(Value=&Waveform(
      Global Amplitude Scale=&Multiply_Float(A=1, B=2)
      "Waveform [1]"=::NewEnumerator0
      "[1] Frequency"=7
      "[1] Amplitude Min/Max"=<0.8 1>
     ))
    ScaleA=true
    ScaleRGB=true
   ) {module=ScaleColor001}
   /SetVariables(Particles.Position=&AddVectorToPosition(Position=<-990.1424 101.7483 60>, Vector=$System.OrbWPOLocal)) {isSetParameters=true}
   /SetVariables(Particles.Pos_02:NiagaraPosition=&AddVectorToPosition(Position=<-404.183 -946.5966 60>, Vector=$System.OrbWPOLocal)) {isSetParameters=true}
   /SetVariables(Particles.Pos_03:NiagaraPosition=&AddVectorToPosition(Position=<792.8474 -691.7074 60>, Vector=$System.OrbWPOLocal)) {isSetParameters=true}
   /SetVariables(Particles.Pos_04:NiagaraPosition=&AddVectorToPosition(Position=<907.5627 505.4716 60>, Vector=$System.OrbWPOLocal)) {isSetParameters=true}
   /SetVariables(Particles.Pos_05:NiagaraPosition=&AddVectorToPosition(Position=<-190.1122 1015.7283 60>, Vector=$System.OrbWPOLocal)) {isSetParameters=true}
  ]
  rendererClasses [&NiagaraMeshRendererProperties]
  renderers [
   {rendererClass=&NiagaraMeshRendererProperties, rendererIndex=0}
   {rendererClass=&NiagaraMeshRendererProperties, rendererIndex=1}
   {rendererClass=&NiagaraMeshRendererProperties, rendererIndex=2}
   {rendererClass=&NiagaraMeshRendererProperties, rendererIndex=3}
   {rendererClass=&NiagaraMeshRendererProperties, rendererIndex=4}
  ]
 }
}
summary {
 systemName=NS_DragonCart_OrbBlast_Looping_V01
 emitters [
  {emitterName=Beam_Glow, rendererClasses [&NiagaraMeshRendererProperties]}
  {emitterName=Ray, rendererClasses [&NiagaraSpriteRendererProperties]}
  {emitterName=CoreBleedLight, rendererClasses [&NiagaraMeshRendererProperties, &NiagaraSpriteRendererProperties]}
  {emitterName=Light_Up, rendererClasses [&NiagaraLightRendererProperties]}
  {emitterName=Impact_Flare, simTarget=GPUComputeSim, rendererClasses [&NiagaraSpriteRendererProperties]}
  {emitterName=Beam_Core, rendererClasses [&NiagaraMeshRendererProperties]}
  {emitterName=Lightning_V01001, rendererClasses [&NiagaraMeshRendererProperties]}
  {emitterName=SmallLightning_Flare, rendererClasses [&NiagaraSpriteRendererProperties]}
  {emitterName=UP_Lightning_Small001, rendererClasses [&NiagaraMeshRendererProperties]}
  {emitterName=DarkSPline001, rendererClasses [&NiagaraMeshRendererProperties]}
  {emitterName=EnergyLeak001, rendererClasses [&NiagaraSpriteRendererProperties]}
 ]
 userVariables [
  User.bIsFrontEnd {type=NiagaraBool}
  User.bIsFrontEndPreview {type=NiagaraBool}
  User.Core_Beam_Z_Offset {type=NiagaraFloat, defaultValue=0, description=""}
  User.Core_Intensity {type=NiagaraFloat, defaultValue=1, description=""}
  User.Intensity {type=NiagaraFloat, defaultValue=1, description=""}
  User.Length {type=NiagaraFloat, defaultValue=1, description=""}
  User.Lightning_Z_Offset {type=NiagaraFloat, defaultValue=-3200, description=""}
  User.SmallLightning_FlareIntensity {type=NiagaraFloat, defaultValue=1, description=""}
  User.SmallLightning_Rate {type=NiagaraFloat, defaultValue=1, description=""}
  User.SpawnSmallLightningLFare {type=NiagaraBool}
  User.UpLight_ExtraIntensity_Mult {type=NiagaraFloat, defaultValue=1, description=""}
 ]
}
