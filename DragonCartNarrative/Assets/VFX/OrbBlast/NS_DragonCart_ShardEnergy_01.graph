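# Niagara system for DragonCart's shard energy VFX, continuously emitting pink-red energy around a shard.
# Four emitters layer curl-noise ember sprites, pulsing flare shockwave sprites, growing bubble meshes and tumbling fragment meshes.
# User.Intensity scales every spawn rate and size curve, and blends ember speed, drag, noise and some lifetimes.
# All emitters use distance-based spawn scaling; the other user parameters are not referenced by the modules shown.
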NiagaraSystem /DragonCartNarrative/Assets/VFX/OrbBlast/NS_DragonCart_ShardEnergy_01
dependencies {
 usedDataInterfaces [&NiagaraDataInterfaceCurve, &NiagaraDataInterfaceColorCurve]
 usedDynamicInputs [&MakeFloatFromVector, &RandomRangeVector]
 usedModules [&SystemState]
}
emitters {
 Embers001 {
  emitterUpdateScript [
   &EmitterState(
    Enable Distance Culling=true
    Max Distance Response=::Awaken
    MaxDistance=32000
    MinDistance=4000
    Module.bApplySpawnCountScaleByDistance=true
    Module.bUseMaxDistance=true
    Module.bUseMinDistance=false
    Scalability Mode=::NewEnumerator1
    Scale Spawn Count=true
   )
   &SpawnRate(SpawnRate=&Multiply_Float(A=100, B=$User.Intensity), Use Spawn Probability=false)
  ]
  particleSpawnScript [
   /SetVariables(Particles.SpriteRotationRate=&RandomRangeFloat(Minimum=-1, Maximum=1, Evaluation Type=::NewEnumerator1)) {isSetParameters=true}
   &InitializeParticle(
    Color=<rgba 5 0 1.0099 1>
    Lifetime=&Lerp_Float(
     A=&RandomRangeFloat(Minimum=0.7, Maximum=1.2)
     B=&RandomRangeFloat(Minimum=0.5, Maximum=0.8)
     Alpha=$User.Intensity
    )
    Sprite Rotation Angle Min=-360
    Sprite Rotation Mode=::NewEnumerator2
    Sprite Size Mode=::NewEnumerator1
    Uniform Sprite Size Max=200
    Uniform Sprite Size Min=140
    UsePositionOffset=false
    Write Material Random=false
   ) {module=InitializeParticle001}
   &ShapeLocation(MaskSpawnGroup=false, Offset Mode=::NewEnumerator0, Override Seed=false, UseOwnerScale=true)
   &AddVelocity(
    Module.Use Falloff Distance=false
    Override Seed=false
    Velocity Mode=::NewEnumerator1
    Velocity Speed=&Lerp_Float(
     A=&RandomRangeFloat(Minimum=1100, Maximum=1500)
     B=&RandomRangeFloat(Minimum=3000, Maximum=4500)
     Alpha=$User.Intensity
    )
   )
  ]
  particleUpdateScript [
   &ParticleState()
   &Drag(
    Drag=&Lerp_Float(
     A=&RandomRangeFloat(Minimum=8, Maximum=12)
     B=&RandomRangeFloat(Minimum=1, Maximum=2)
     Alpha=$User.Intensity
    )
    Use Linear Drag=true
    Use Rotational Drag=true
   )
   &CurlNoiseForce(
    Bias Noise=false
    Noise Frequency=30
    Noise Strength=&Lerp_Float(A=1000, B=6000, Alpha=$User.Intensity)
    Pan Noise=true
    Pan Noise Field=<0 0 1>
   )
   &SolveForcesAndVelocity(Clamp Velocity=false, Limit Acceleration=false)
   &ScaleSpriteSize(Scale Factor=&Vector2DFromFloat(Value=&FloatFromCurve(
      FloatCurve=DataInterface() {propertyValues {Curve {keys [
          {interpMode=RCIM_Cubic, leaveTangent=1.7285, tangentMode=RCTM_Break, time=0, value=0.2654}
          {arriveTangent=-0.2654, interpMode=RCIM_Cubic, leaveTangent=-0.2654, time=0.4, value=0.8003}
          <at 1 0>
         ]}}}
      Scale Curve=&Multiply_Float(A=2, B=$User.Intensity)
     )))
   &SpriteRotationRate(Rotation Rate=$Particles.SpriteRotationRate)
  ]
  rendererClasses [&NiagaraSpriteRendererProperties]
  renderers [{rendererClass=&NiagaraSpriteRendererProperties}]
 }
 FlareLightShockwave001 {
  emitterUpdateScript [
   &EmitterState(
    Enable Distance Culling=true
    Max Distance Response=::Awaken
    MaxDistance=32000
    MinDistance=4000
    Module.bApplySpawnCountScaleByDistance=true
    Module.bUseMaxDistance=true
    Module.bUseMinDistance=false
    Scalability Mode=::NewEnumerator1
    Scale Spawn Count=true
   )
   &SpawnRate(SpawnRate=&Multiply_Float(A=4, B=$User.Intensity), Use Spawn Probability=false)
  ]
  particleSpawnScript [&InitializeParticle(
    Color=<rgba 5 0 0.9073 1>
    Lifetime=&Lerp_Float(A=1, B=0.6, Alpha=$User.Intensity)
    Sprite Size Mode=::NewEnumerator3
    Uniform Sprite Size=3000
    UsePositionOffset=true
    Write Material Random=true
   )]
  particleUpdateScript [
   &ParticleState()
   &CameraOffset(Camera Offset Amount=&FloatFromCurve(
     FloatCurve=DataInterface() {propertyValues {Curve {keys [
         {arriveTangent=4.0899, interpMode=RCIM_Cubic, leaveTangent=4.0899, tangentMode=RCTM_User, time=0, value=0.5}
         {arriveTangent=-3.6426, interpMode=RCIM_Cubic, leaveTangent=-3.6426, tangentMode=RCTM_User, time=0.2, value=1}
         {arriveTangent=5.165, interpMode=RCIM_Cubic, leaveTangent=5.165, tangentMode=RCTM_User, time=0.4, value=0.3}
         <at 0.75 0>
        ]}}}
     Scale Curve=200
    ))
   &DynamicMaterialParameters(
    Index 0 Param 1=1
    Index 0 Param 2=&FloatFromCurve(FloatCurve=DataInterface() {propertyValues {Curve {keys [{interpMode=RCIM_Cubic, tangentMode=RCTM_User, time=0, value=1}, {interpMode=RCIM_Cubic, tangentMode=RCTM_User, time=1, value=0}]}}})
    Module.Param0WriteEnabled=true
    Module.Param1WriteEnabled=true
    Module.Param2WriteEnabled=false
    Module.Param3WriteEnabled=false
   )
   &ScaleSpriteSize(
    Uniform Curve Scale=$User.Intensity
    Uniform Curve Sprite Scale=DataInterface() {propertyValues {Curve {keys [
        {arriveTangent=21.6769, interpMode=RCIM_Cubic, time=0, value=0.1}
        {arriveTangentWeight=0.2899, interpMode=RCIM_Cubic, leaveTangentWeight=0.1115, tangentMode=RCTM_User, time=0.5, value=1}
       ]}}}
   )
   &ScaleColor(
    Scale Alpha=&FloatFromCurve(FloatCurve=DataInterface() {propertyValues {Curve {keys [
         {arriveTangent=21.6769, interpMode=RCIM_Cubic, leaveTangent=21.6769, tangentMode=RCTM_User, time=0, value=0}
         <at 0.1 1>
         {arriveTangent=-1.1111, interpMode=RCIM_Cubic, leaveTangent=-1.1111, time=0.35, value=1}
         <at 1 0>
        ]}}})
    ScaleA=true
    ScaleRGB=true
   )
  ]
  rendererClasses [&NiagaraSpriteRendererProperties]
  renderers [{rendererClass=&NiagaraSpriteRendererProperties}]
 }
 ShardBubble001 {
  emitterUpdateScript [
   &EmitterState(
    Enable Distance Culling=true
    Max Distance Response=::Awaken
    MaxDistance=32000
    MinDistance=4000
    Module.bApplySpawnCountScaleByDistance=true
    Module.bUseMaxDistance=true
    Module.bUseMinDistance=false
    Scalability Mode=::NewEnumerator1
    Scale Spawn Count=true
   )
   &SpawnRate(SpawnRate=&Multiply_Float(A=3, B=$User.Intensity), Use Spawn Probability=false)
  ]
  particleSpawnScript [
   &InitializeParticle(
    Color=<rgba 3.2292 0 0.1842 1>
    Lifetime=&Lerp_Float(
     A=&RandomRangeFloat(Minimum=1, Maximum=1.2)
     B=&RandomRangeFloat(Minimum=0.9, Maximum=1.1)
     Alpha=$User.Intensity
    )
    Mesh Scale Mode=::NewEnumerator1
    Mesh Uniform Scale Max=36
    Mesh Uniform Scale Min=33
    UsePositionOffset=false
    Write Material Random=true
   )
   &InitialMeshOrientation(Use Rotation Vector=false)
  ]
  particleUpdateScript [
   &ParticleState()
   &ScaleMeshSize(Scale Factor=&VectorFromFloat(Value=&FloatFromCurve(
      FloatCurve=DataInterface() {propertyValues {Curve {keys [
          {interpMode=RCIM_Cubic, leaveTangent=0.8229, tangentMode=RCTM_Break, time=0, value=0}
          {interpMode=RCIM_Cubic, tangentMode=RCTM_Break, time=1, value=1}
         ]}}}
      Scale Curve=$User.Intensity
     )))
   &ScaleColor(
    Scale Alpha=&FloatFromCurve(FloatCurve=DataInterface() {propertyValues {Curve {keys [
         {interpMode=RCIM_Cubic, tangentMode=RCTM_Break, time=0, value=0}
         {arriveTangent=-0.0848, interpMode=RCIM_Cubic, leaveTangent=-0.0848, tangentMode=RCTM_User, time=0.2, value=1}
         {arriveTangent=-0.0112, interpMode=RCIM_Cubic, leaveTangent=-0.0112, tangentMode=RCTM_User, time=0.5, value=1}
         {interpMode=RCIM_Cubic, tangentMode=RCTM_Break, time=1, value=0}
        ]}}})
    Scale RGB=&VectorFromFloat(Value=7)
    ScaleA=true
    ScaleRGB=true
   )
   &DynamicMaterialParameters(
    Index 0 Param 1=&FloatFromCurve(FloatCurve=DataInterface() {propertyValues {Curve {keys [{interpMode=RCIM_Cubic, tangentMode=RCTM_Break, time=0, value=0}, {interpMode=RCIM_Cubic, tangentMode=RCTM_Break, time=1, value=0.8}]}}})
    Module.Index0Float0Write=true
    Module.Index0Float1Write=true
    Module.Index0Float2Write=true
    Module.Index0Float3Write=true
   )
  ]
  rendererClasses [&NiagaraMeshRendererProperties]
  renderers [{rendererClass=&NiagaraMeshRendererProperties, rendererIndex=0}]
 }
 TinyFrags001 {
  emitterUpdateScript [
   &EmitterState(
    Enable Distance Culling=true
    Max Distance Response=::Awaken
    MaxDistance=32000
    MinDistance=4000
    Module.bApplySpawnCountScaleByDistance=true
    Module.bUseMaxDistance=true
    Module.bUseMinDistance=false
    Scalability Mode=::NewEnumerator1
    Scale Spawn Count=true
   )
   &SpawnRate(SpawnRate=&Multiply_Float(A=30, B=$User.Intensity), Use Spawn Probability=false)
  ]
  particleSpawnScript [
   &InitializeParticle(
    Lifetime=&Multiply_Float(A=1, B=&RandomRangeFloat(Minimum=0.5, Maximum=1))
    Mesh Scale Mode=::NewEnumerator1
    UsePositionOffset=false
    Write Material Random=false
   ) {module=InitializeParticle001}
   &ShapeLocation(MaskSpawnGroup=false, Offset Mode=::NewEnumerator0, Override Seed=false, UseOwnerScale=true)
   &AddVelocity(
    Module.Use Falloff Distance=false
    Override Seed=false
    Velocity Mode=::NewEnumerator1
    Velocity Speed=&RandomRangeFloat(Minimum=3000, Maximum=4500)
   )
   &InitialMeshOrientation(Use Rotation Vector=false)
  ]
  particleUpdateScript [
   &ParticleState()
   &Drag(Drag=&RandomRangeFloat(Minimum=4, Maximum=6), Use Linear Drag=true, Use Rotational Drag=true)
   &CurlNoiseForce(Bias Noise=false, Noise Frequency=30, Noise Strength=1500, Pan Noise=true, Pan Noise Field=<0 0 1>)
   &SolveForcesAndVelocity(Clamp Velocity=false, Limit Acceleration=false)
   &ScaleColor(
    Scale Alpha=&FloatFromCurve(FloatCurve=DataInterface() {propertyValues {Curve {keys [
         <at 0 0 RCIM_Cubic>
         {arriveTangent=1.4286, interpMode=RCIM_Cubic, leaveTangent=1.4286, time=0.2, value=1}
         {arriveTangent=-1.25, interpMode=RCIM_Cubic, leaveTangent=-1.25, time=0.7, value=1}
         <at 1 0>
        ]}}})
    Scale RGB=&MakeVectorFromLinearColorRGB(Color=&ColorFromCurve(ColorCurve=DataInterface() {propertyValues {
        AlphaCurve {keys [<at 0 1 RCIM_Cubic>,<at 1 0 RCIM_Cubic>]}
        BlueCurve {keys [<at 0.1928 50 RCIM_Cubic>,<at 1 1 RCIM_Cubic>]}
        GreenCurve {keys [<at 0.1928 50 RCIM_Cubic>,<at 1 1 RCIM_Cubic>]}
        RedCurve {keys [<at 0.1928 50 RCIM_Cubic>,<at 1 1 RCIM_Cubic>]}
       }}))
    ScaleA=true
    ScaleRGB=true
   )
   &UpdateMeshOrientation(Rotation Rate=0.5, Rotation Vector=&RandomVector(Vector Scale=1)) {module=UpdateMeshOrientation001}
   &ScaleMeshSize(Scale Factor=&VectorFromFloat(Value=&FloatFromCurve(
      FloatCurve=DataInterface() {propertyValues {Curve {keys [
          {arriveTangent=9.5937, interpMode=RCIM_Cubic, leaveTangent=9.5937, tangentMode=RCTM_User, time=0, value=0}
          <at 0.2 1>
          {arriveTangentWeight=0.2134, interpMode=RCIM_Cubic, leaveTangentWeight=0.2134, tangentMode=RCTM_User, time=0.6, value=1}
          {arriveTangent=-5.5237, interpMode=RCIM_Cubic, leaveTangent=-5.5237, tangentMode=RCTM_User, time=1, value=0}
         ]}}}
      Scale Curve=&Multiply_Float(A=4, B=$User.Intensity)
     )))
  ]
  rendererClasses [&NiagaraMeshRendererProperties]
  renderers [{rendererClass=&NiagaraMeshRendererProperties, rendererIndex=0}]
 }
}
summary {
 systemName=NS_DragonCart_ShardEnergy_01
 emitters [
  {emitterName=ShardBubble001, rendererClasses [&NiagaraMeshRendererProperties]}
  {emitterName=FlareLightShockwave001, rendererClasses [&NiagaraSpriteRendererProperties]}
  {emitterName=Embers001, rendererClasses [&NiagaraSpriteRendererProperties]}
  {emitterName=TinyFrags001, rendererClasses [&NiagaraMeshRendererProperties]}
 ]
 userVariables [
  User.DoShcokwave {type=NiagaraBool, defaultValue=true}
  User.Intensity {type=NiagaraFloat, defaultValue=1, description=""}
  User.Life_Mult {type=NiagaraFloat, defaultValue=1, description=""}
  User.ShockWave {type=NiagaraInt32, defaultValue=1}
  User.Size_Mult {type=NiagaraFloat, defaultValue=1, description=""}
  User.SpawnRate_Mult {type=NiagaraFloat, defaultValue=1, description=""}
  User.ZeroPointPosition {type=NiagaraPosition, description="", defaultValue {}}
 ]
}
